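# Conversation database for the Jekyll NPC, providing his greeting dialogue and access to shared NPC conversations.
# The entry point leads to one speech node with a default line and contextual variants for swamp visitors, first-ever conversations and specific Slurp or BlueFlames outfits.
# Each variant carries a priority from a named curve row.
# The speech then branches through link nodes to the shared NPC library quests and basic services entry points.
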ConversationDatabase /NPCLibrary/NPCs/Jekyll/Jekyll/NPCLibrary_Conversation_Jekyll
entry_points [{tag=Conversation.Entry.Jekyll, destinations [#1]}]
nodes [
 ConversationEntryPointNode_7 {
  id=#1
  class=ConversationEntryPointNode
  outputs [#2]
  properties {entryTag {tagName=Conversation.Entry.Jekyll}}
 }
 Speech_C_4 {
  id=#2
  class=Speech_C
  outputs [#3, #4]
  properties {
   nodeName=""
   generalConfig {
    defaultMessage="Careful. I'm not feeling.. quite <b>myself</> these days."
    contextualMessages [
     {
      message="Ah! You still have the dank of swamp-muck about you. Please, let me take a sample."
      requirementMatchPolicy=RequireAny
      contextRequirements [
       {requirement {class=FortControllerRequirement_HasVisitedPOI, properties {pOITag {tagName=Athena.Location.POI.SlurpySwamp}}}}
       {requirement {class=FortControllerRequirement_HasVisitedPOI, properties {pOITag {tagName=Athena.Location.UnnamedPOI.SwampVille}}}}
       {requirement {class=FortControllerRequirement_HasVisitedPOI, properties {pOITag {tagName=Athena.Location.Biome.Swamp}}}}
      ]
      priority {curve {rowName=Tandem.Speech.Priority.VisitedSpecificPOI}}
     }
     {
      message="Ah, are you my new test subj-- assistant? No?"
      contextRequirements [{requirement {class=FortControllerRequirement_IsFirstEverConversationWithNPC}}]
      priority {curve {rowName=Tandem.Speech.Priority.FirstEverInteraction}}
     }
     {
      message="Ahh a fellow victim of science. Hello, <b>my brethren.</>"
      requirementMatchPolicy=RequireAny
      contextRequirements [{requirement {
         class=FortControllerRequirement_HasCID
         properties {allowedCIds [
           /BRCosmetics/Athena/Items/Cosmetics/Characters/CID_708_Athena_Commando_M_SoldierSlurp
           /BRCosmetics/Athena/Items/Cosmetics/Characters/CID_709_Athena_Commando_F_BandolierSlurp
           /BRCosmetics/Athena/Items/Cosmetics/Characters/CID_710_Athena_Commando_M_FishheadSlurp
           /BRCosmetics/Athena/Items/Cosmetics/Characters/CID_717_Athena_Commando_F_BlueFlames
           /BRCosmetics/Athena/Items/Cosmetics/Characters/CID_716_Athena_Commando_M_BlueFlames
           /BRCosmetics/Athena/Items/Cosmetics/Characters/CID_570_Athena_Commando_M_SlurpMonster
          ]}
        }}]
      priority {curve {rowName=Tandem.Speech.Priority.OutfitReaction}}
     }
    ]
   }
  }
 }
 ConversationLinkNode_0 {id=#4, class=ConversationLinkNode, properties {remoteEntryTag {tagName=Conversation.Entry.BasicServices}}}
 ConversationLinkNode_1 {id=#3, class=ConversationLinkNode, properties {remoteEntryTag {tagName=Conversation.Entry.NPCLibrary.Quests}}}
]
